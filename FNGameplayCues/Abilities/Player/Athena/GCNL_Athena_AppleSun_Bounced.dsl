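; Looping gameplay cue for the player "AppleSun" bounced ability effect, handling its particles and audio.
; On application, it traces a short distance below the target and, on a hit, moves the first particle component to that point and enlarges it.
; Loop start caches the first audio component and target.
; Tick drives the audio "Velocity" parameter from vertical speed for locally controlled pawns; others get full value, then ticking stops.
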(event EventOnApplication(Cascade)
  (params (MyTarget :type Actor%20Object%20Reference) (Parameters :type Gameplay%20Cue%20Parameters%20Structure%20%28by%20ref%29) (ParticleComponents :type Array%20of%20Cascade%20Particle%20System%20Component%20Object%20References) (AudioComponents :type Array%20of%20Audio%20Component%20Object%20References) (BurstCameraShakeInstance :type Camera%20Shake%20Base%20Object%20Reference) (BurstDecalInstance :type Decal%20Actor%20Object%20Reference))
  (bind _output (Utilities|Array|Get(acopy) ParticleComponents))
  (bind (_outhit _returnvalue) (Collision|LineTraceForObjects 0 (Transformation|GetActorLocation MyTarget) (- (Transformation|GetActorLocation MyTarget) "0, 0,250.000000") (Utilities|Array|MakeArray "ObjectTypeQuery2") true))
  (if _returnvalue
    (Transformation|SetWorldLocation _output (.Location (Collision|BreakHitResult _outhit)) false true)
    (Transformation|SetWorldScale3D _output "1.250000,1.250000,1.250000")))

(event EventTick
  (params (DeltaSeconds :type Float%20%28single-precision%29))
  (bind _target (Variables|Default|GetTarget)
    (:then
      (bind _asfort_pawn (Utilities|Casting|CastToFortPawn _target))
      (bind _inuseloopaudio (Variables|Default|GetInUseLoopAudio)
        (:then
          (Utilities|FlowControl|Sequence
            (:then_0
              (bind (_returnvalue_x _returnvalue_y _returnvalue_z) (Math|Vector|BreakVector (Transformation|GetVelocity _target)))
              (Audio|Parameter|SetFloatParameter _inuseloopaudio "Velocity" (Math|Float|SelectFloat (Math|Float|MapRangeClamped (Math|Float|Absolute(Float) _returnvalue_z) 0.0 1e+03 0.0 1.0) 1.0 (Pawn|IsLocallyControlled _asfort_pawn)) :!class /Script/AudioExtensions.AudioParameterControllerInterface))
            (:then_1
              (if (not (Pawn|IsLocallyControlled _asfort_pawn))
                (Actor|Tick|SetActorTickEnabled false)))))
        (:else)))
    (:else)))

(event EventOnLoopingStart(Cascade)
  (params (MyTarget :type Actor%20Object%20Reference) (Parameters :type Gameplay%20Cue%20Parameters%20Structure%20%28by%20ref%29) (ParticleComponents :type Array%20of%20Cascade%20Particle%20System%20Component%20Object%20References) (AudioComponents :type Array%20of%20Audio%20Component%20Object%20References))
  (Variables|Default|SetInUseLoopAudio (Utilities|Array|Get(acopy) AudioComponents))
  (Variables|Default|SetTarget MyTarget))

(fn ConstructionScript
  (|CallParentFunction))
