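; Looping gameplay cue for the tether ability that attaches skis to a player pawn's feet and drives surface-dependent effects.
; On application it attaches both skis and scales in the left one after a delay.
; Tick feeds speed and turn values by water, land or falling state into layered audio, particles and controller rumble.
; On water, local players viewing a wildlife pawn get camera-angle-gated lens droplets; removal fades audio and clears the lens effect.
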(event Custom|ScaleInSkisForPlayer
  (bind _leftski (Variables|Default|GetLeftSki))
  ;[1 1 1] Ski scaling event
  (bind (_direction _skiscalar) (Timeline|ScaleInSkis Play)
    (:Update
      (Transformation|SetWorldScale3D _leftski (* "1.000000,1.000000,1.000000" _skiscalar)))
    (:Finished
      (Transformation|SetWorldScale3D _leftski "1.000000,1.000000,1.000000"))))

(event Custom|StartFF
  ;[1 1 1] Handle controller vibration
  (bind _owningfpp (Variables|Default|GetOwningFPP)
    (:then
      _)
    (:else
      _))
  (Game|Feedback|PlayDynamicForceFeedback (Fortnite|GetFortPlayerControllerfromActor _owningfpp) (Variables|Default|GetForceFeedbackIntensity) -1 true (or (< (Variables|Default|GetTurnSpeed) -0.15) (not (Variables|Default|GetisInWater))) true (or (> (Variables|Default|GetTurnSpeed) 0.15) (not (Variables|Default|GetisInWater)))))

(event EventOnApplication(Cascade)
  (params (MyTarget :type Actor%20Object%20Reference) (Parameters :type Gameplay%20Cue%20Parameters%20Structure%20%28by%20ref%29) (ParticleComponents :type Array%20of%20Cascade%20Particle%20System%20Component%20Object%20References) (AudioComponents :type Array%20of%20Audio%20Component%20Object%20References) (BurstCameraShakeInstance :type Camera%20Shake%20Base%20Object%20Reference) (BurstDecalInstance :type Decal%20Actor%20Object%20Reference))
  ;[1 1 1] On Application
  (CallFunction|InitializeTetherGCNL MyTarget)
  (Transformation|SetWorldScale3D (Variables|Default|GetLeftSki) "0.000000,1.000000,0.000000")
  (bind _returnvalue (Utilities|Time|SetTimerbyEvent (EventDispatchers|CreateEvent :!event Custom|ScaleInSkisForPlayer) (Variables|Default|GetSkiScaleInDelay)))
  (Variables|Default|SetScaleInSkisTimerHandle _returnvalue))

(event Custom|UpdateFF
  ;[1 1 1] Handle controller vibration
  (Variables|Default|GetOwningFPP
    (:then
      _)
    (:else
      _))
  (Game|Feedback|PlayDynamicForceFeedback (Fortnite|GetFortPlayerControllerfromActor (stale Custom|StartFF _owningfpp)) (Variables|Default|GetForceFeedbackIntensity) -1 true (or (< (Variables|Default|GetTurnSpeed) -0.15) (not (Variables|Default|GetisInWater))) true (or (> (Variables|Default|GetTurnSpeed) 0.15) (not (Variables|Default|GetisInWater)))))

(event Custom|StopFF
  ;[1 1 1] Handle controller vibration
  (Game|Feedback|PlayDynamicForceFeedback (Fortnite|GetFortPlayerControllerfromActor (stale Custom|StartFF _owningfpp)) (Variables|Default|GetForceFeedbackIntensity) -1 true (or (< (Variables|Default|GetTurnSpeed) -0.15) (not (Variables|Default|GetisInWater))) true (or (> (Variables|Default|GetTurnSpeed) 0.15) (not (Variables|Default|GetisInWater)))))

(event EventOnRemoval(Cascade)
  (params (MyTarget :type Actor%20Object%20Reference) (Parameters :type Gameplay%20Cue%20Parameters%20Structure%20%28by%20ref%29) (ParticleComponents :type Array%20of%20Cascade%20Particle%20System%20Component%20Object%20References) (AudioComponents :type Array%20of%20Audio%20Component%20Object%20References) (BurstCameraShakeInstance :type Camera%20Shake%20Base%20Object%20Reference) (BurstDecalInstance :type Decal%20Actor%20Object%20Reference))
  ;[1 1 1] On Remove
  (Utilities|FlowControl|Sequence
    (:then_0
      (Components|Activation|SetActive (Variables|Default|GetWaterFX) false false)
      (Audio|Components|Audio|FadeOut (Variables|Default|GetFortLayeredAudio) 1.0 0.0 "Linear" :!class /Script/Engine.AudioComponent))
    (:then_1
      (bind _asactor (Utilities|Casting|CastToActor (Variables|Default|GetCameraLensEffectInterface)))
      (Actor|DestroyActor _asactor))))

(event Custom|InitializeTetherGCNL
  (params (Object :type Object%20Reference))
  (bind _leftski (Variables|Default|GetLeftSki))
  (bind _rightski (Variables|Default|GetRightSki))
  ;[1 1 1] Init (This gets called from On Application and While Active function)
  (bind _asfort_player_pawn (Utilities|Casting|CastToFortPlayerPawn Object))
  (bind _output_get (Variables|Default|SetOwningFPP _asfort_player_pawn))
  (Transformation|AttachComponentToComponent _leftski (Variables|/Script/Engine.Character|GetMesh _output_get) "foot_l" "SnapToTarget" "SnapToTarget" "KeepRelative" true)
  (Transformation|SetRelativeRotation _leftski "-90.000000,0.000000, 0" false false)
  (Transformation|AttachComponentToComponent _rightski (Variables|/Script/Engine.Character|GetMesh _output_get) "foot_r" "SnapToTarget" "SnapToTarget" "KeepRelative" true)
  (Transformation|SetRelativeRotation _rightski "-90.000000,180.000000, 0" false false)
  (CallFunction|StartFF)
  (bind _asactor (Utilities|Casting|CastToActor (Variables|Default|GetCameraLensEffectInterface)))
  (Actor|DestroyActor _asactor))

(event EventBeginPlay
  ;[1 1 1] Event BeginPlay
  (|CallParentFunction)
  (Actor|Tick|SetTickGroup "TG_DuringPhysics"))

(event EventTick
  (params (DeltaSeconds :type Float%20%28single-precision%29))
  (bind _owningfpp (Variables|Default|GetOwningFPP))
  (bind _returnvalue (Transformation|GetVelocity _owningfpp))
  (bind _owningfpp_2 (Variables|Default|GetOwningFPP))
  (bind _returnvalue_2 (Fortnite|GetFortPlayerControllerfromActor _owningfpp_2))
  (bind _landfx (Variables|Default|GetLandFX))
  (bind _waterfx (Variables|Default|GetWaterFX))
  (bind _waterfx_2 (Variables|Default|GetWaterFX))
  (bind _cameralenseffectinterface (Variables|Default|GetCameraLensEffectInterface))
  (bind _cameralenseffectinterface_2 (Variables|Default|GetCameraLensEffectInterface))
  (bind _cameralenseffectinterface_3 (Variables|Default|GetCameraLensEffectInterface))
  ;[1 1 1] Event Tick
  (Variables|Default|SetNormalizedSpeed (Math|Float|MapRangeClamped (Math|Vector|VectorLength _returnvalue) 2e+02 2.5e+03 0.0 1.0))
  (Variables|Default|SetTurnSpeed (Math|Vector|DotProduct (Math|Vector|Normalize _returnvalue) (Transformation|GetForwardVector (Variables|/Script/Engine.Character|GetMesh _owningfpp))))
  (Utilities|UpdateBasedonMovementType
    (:UpdateWater
      ;[1 1 1] Update While in Water
      (CallFunction|UpdateAudio 0)
      (bind _returnvalue_3 (WaterBody|SetCommonWaterFXParams (Variables|Default|GetOwningFPP) _waterfx true true true))
      (Components|Activation|SetActive _waterfx _returnvalue_3 false)
      (Components|Activation|SetActive (Variables|Default|GetLandFX) false false)
      (Effects|Components|ParticleSystem|SetFloatParameter _waterfx_2 "Speed" (Variables|Default|GetNormalizedSpeed))
      (Effects|Components|ParticleSystem|SetFloatParameter _waterfx_2 "TurnSpeed" (Variables|Default|GetTurnSpeed))
      (Utilities|FlowControl|Sequence
        (:then_0
          ;[1 1 1] Set up camera lens effect
          (if (Pawn|IsLocallyControlled _owningfpp_2)
            (Utilities|Casting|CastToNPC_Pawn_Wildlife_Parent (Pawn|GetViewTarget _returnvalue_2)
              (:then
                (LensEffect|IsValidCameraLens _cameralenseffectinterface_3
                  (:Valid
                    (Effects|Components|ParticleSystem|SetFloatParameter (CameraLensEffect|GetPrimaryParticleComponent (Variables|Default|GetCameraLensEffectInterface)) "SpawnRate" (Math|Conversions|ToFloat(Boolean) (< (Math|Vector|DotProduct (Transformation|GetForwardVector (CameraLensEffect|GetPrimaryParticleComponent (Variables|Default|GetCameraLensEffectInterface))) (Transformation|GetActorForwardVector (Variables|Default|GetOwningFPP))) -0.5))) ; Spawn camera droplets when camera is behind player
                  )
                  (:Invalid
                    (bind _returnvalue_4 (Camera|AddGenericCameraLensEffect (Variables|/Script/Engine.PlayerController|GetPlayerCameraManager _returnvalue_2) "/FNGameplayCues/Effects/Fort_Effects/Camera_Effects/B_CameraLens_Boat_Droplets.B_CameraLens_Boat_Droplets_C"))
                    (Variables|Default|SetCameraLensEffectInterface _returnvalue_4))))
              (:CastFailed
                (bind _asactor (Utilities|Casting|CastToActor _cameralenseffectinterface_3))
                (Actor|DestroyActor _asactor)))))
        (:then_1
          (Variables|Default|SetForceFeedbackIntensity (* (+ (* (Math|Float|Absolute(Float) (Variables|Default|GetTurnSpeed)) 0.35) (Math|Float|Clamp(Float) (Variables|Default|GetNormalizedSpeed) 0.0 0.5)) 0.15)))))
    (:UpdateLand
      ;[1 1 1] Update While On Land
      (CallFunction|UpdateAudio 1)
      (Components|Activation|SetActive (Variables|Default|GetWaterFX) false false)
      (Components|Activation|SetActive (Variables|Default|GetLandFX) (> (Math|Vector|VectorLength (Transformation|GetVelocity (Variables|Default|GetOwningFPP))) 1.5e+02) false)
      (Effects|Components|ParticleSystem|SetFloatParameter _landfx "Speed" (Variables|Default|GetNormalizedSpeed))
      (Effects|Components|ParticleSystem|SetFloatParameter _landfx "TurnSpeed" (Variables|Default|GetTurnSpeed))
      (Variables|Default|SetForceFeedbackIntensity (* (+ (* (Math|Float|Absolute(Float) (Variables|Default|GetTurnSpeed)) 0.3) (Math|Float|Clamp(Float) (Variables|Default|GetNormalizedSpeed) 0.0 0.5)) 0.2))
      (LensEffect|IsValidCameraLens _cameralenseffectinterface
        (:Valid
          (Effects|Components|ParticleSystem|SetFloatParameter (CameraLensEffect|GetPrimaryParticleComponent _cameralenseffectinterface) "SpawnRate" 0.0))))
    (:UpdateFalling
      ;[1 1 1] Update While Falling
      (CallFunction|UpdateAudio 2)
      (Components|Activation|SetActive (Variables|Default|GetWaterFX) false false)
      (Variables|Default|SetForceFeedbackIntensity 0.0)
      (LensEffect|IsValidCameraLens _cameralenseffectinterface_2
        (:Valid
          (Effects|Components|ParticleSystem|SetFloatParameter (CameraLensEffect|GetPrimaryParticleComponent _cameralenseffectinterface_2) "SpawnRate" 0.0))))
    (:Update
      (CallFunction|UpdateFF))))

(event OnOneshotEvent(FortLayeredAudio)
  (params (AudioComponent :type Audio%20Component%20Object%20Reference) (ParameterName :type Name :default "None") (OneshotIndex :type Integer))
  (Audio|Parameter|SetIntegerParameter AudioComponent "IsInWater" (Math|Conversions|ToInteger(Boolean) (Variables|Default|GetisInWater)) :!class /Script/AudioExtensions.AudioParameterControllerInterface))

(fn OnCeaseRelevant
  (params (MyTarget :type Actor%20Object%20Reference) (Parameters :type Gameplay%20Cue%20Parameters%20Structure%20%28by%20ref%29))
  (bind _scaleinskistimerhandle (Variables|Default|GetScaleInSkisTimerHandle))
  (Audio|Components|Audio|FadeOut (Variables|Default|GetFortLayeredAudio) 1.0 0.0 "Linear" :!class /Script/Engine.AudioComponent)
  (if (Utilities|Time|IsValidTimerHandle _scaleinskistimerhandle)
    (Utilities|Time|ClearandInvalidateTimerbyHandle 0 _scaleinskistimerhandle)
    (else
      _))
  (return true))

(fn UpdateAudio
  (params (Surface :type Integer))
  (Class|FortLayeredAudioComponent|SetFloatParameterExt (Variables|Default|GetFortLayeredAudio) "Surface" (/ (Math|Conversions|ToFloat(Integer) Surface) 3.0))
  (Class|FortLayeredAudioComponent|SetFloatParameterExt (Variables|Default|GetFortLayeredAudio) "Velocity" (Variables|Default|GetNormalizedSpeed))
  (Class|FortLayeredAudioComponent|SetFloatParameterExt (Variables|Default|GetFortLayeredAudio) "Steering" (Math|Float|Absolute(Float) (Variables|Default|GetTurnSpeed))))

(fn ConstructionScript)

(fn OnBecomeRelevant
  (params (MyTarget :type Actor%20Object%20Reference) (Parameters :type Gameplay%20Cue%20Parameters%20Structure%20%28by%20ref%29))
  (CallFunction|InitializeTetherGCNL MyTarget)
  (Actor|Tick|SetActorTickEnabled true)
  (return false))
