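; Gameplay cue notify for the Athena player cage trap, firing one-shot activation feedback when the trap is triggered.
; On execute it reads the target attach component from the cue parameters and plays the trap fire sound at that component's world location.
; It also spawns the build particle system there, matching the component's rotation and scale, then returns the parent implementation's result.
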(fn OnExecute
  (params (MyTarget :type Actor%20Object%20Reference) (Parameters :type Gameplay%20Cue%20Parameters%20Structure%20%28by%20ref%29))
  (bind (_normalizedmagnitude _rawmagnitude _effectcontext _matchedtagname _originaltag _aggregatedsourcetags _aggregatedtargettags _location _normal _instigator _effectcauser _sourceobject _physicalmaterial _gameplayeffectlevel _abilitylevel _targetattachcomponent) (Ability|GameplayCue|BreakGameplayCueParameters Parameters))
  (bind _returnvalue (Transformation|GetWorldLocation _targetattachcomponent))
  (Audio|PlaySoundatLocation _targetattachcomponent (Variables|Default|GetTrap_Fire_Sound) _returnvalue)
  (Effects|SpawnEmitteratLocation _targetattachcomponent (Variables|Default|GetParticleSystem_Build) _returnvalue (Transformation|GetWorldRotation _targetattachcomponent) (Math|Vector|MakeVector (.x (Transformation|GetWorldScale _targetattachcomponent)) (.y (Transformation|GetWorldScale _targetattachcomponent)) (.z (Transformation|GetWorldScale _targetattachcomponent))) true "None" true)
  (return (|CallParentFunction MyTarget Parameters)))
