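; Looping gameplay cue for the Low Gravity stone effect on a player pawn, handling its FX and audio.
; When relevant, it caches the pawn, reads duration and warning lead time from AthenaGameData, sets an ExpireTell timer, attaches looping FX and fades in audio flagged for local control.
; On removal it stops the loop, plays end FX and sound, and fades audio out.
; ExpireTell and ClearExpireTell are empty; nothing here calls the warning sound event.
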(event Custom|ExpireTell)

(event Custom|PlayExpirationSound
  (Audio|SpawnSoundAttached "/FNGameplayCues/Sounds/Fort_GamePlay_Sounds/Player/LowGravityStones/LowGravityStone_Warning_Cue.LowGravityStone_Warning_Cue" (Variables|/Script/Engine.Actor|GetRootComponent (Variables|FortGameplayCueNotifyLoop_LowGravity|GetPawn))))

(event Custom|ClearExpireTell)

(event EventBeginPlay)

(event GameplayCueNotify|EventHandleGameplayCue
  (params (MyTarget :type Actor%20Object%20Reference) (EventType :type EGameplayCueEvent%20Enum :default "OnActive") (Parameters :type Gameplay%20Cue%20Parameters%20Structure%20%28by%20ref%29))
  (bind _output_get (Variables|Default|SetMyTargetCached MyTarget))
  (switch Utilities|FlowControl|Switch|SwitchonEGameplayCueEvent EventType))

(event Collision|EventActorBeginOverlap
  (params (OtherActor :type Actor%20Object%20Reference)))

(fn Activated
  (params (PlayerPawn :type Actor%20Object%20Reference))
  (bind _asfort_pawn (Utilities|Casting|CastToFortPawn PlayerPawn))
  (Variables|FortGameplayCueNotifyLoop_LowGravity|SetPawn _asfort_pawn)
  (bind _outxy (DataTable|EvaluateCurveTableRow "/Game/Athena/Balance/DataTables/AthenaGameData.AthenaGameData" "Default.LowGravRock.Duration")
    (:RowFound
      (Variables|FortGameplayCueNotifyLoop_LowGravity|SetDuration _outxy))
    (:RowNotFound
      _))
  (bind _outxy_2 (DataTable|EvaluateCurveTableRow "/Game/Athena/Balance/DataTables/AthenaGameData.AthenaGameData" "Default.LowGravRock.ExpireTellDuration")
    (:RowFound
      (Variables|FortGameplayCueNotifyLoop_LowGravity|SetExpirationTellDuration _outxy_2))
    (:RowNotFound
      _))
  (CallFunction|ClearExpireTell)
  (bind _returnvalue (Utilities|Time|SetTimerbyEvent (EventDispatchers|CreateEvent :!event ExpireTell) (- (Variables|FortGameplayCueNotifyLoop_LowGravity|GetDuration) (Variables|FortGameplayCueNotifyLoop_LowGravity|GetExpirationTellDuration))))
  (Variables|Default|SetExpireTellDelayTimer _returnvalue)
  (Variables|FortGameplayCueNotifyLoop_LowGravity|SetStartTime (Utilities|Time|GetTimeSeconds (Variables|FortGameplayCueNotifyLoop_LowGravity|GetPawn))))

(fn OnCeaseRelevant
  (params (MyTarget :type Actor%20Object%20Reference) (Parameters :type Gameplay%20Cue%20Parameters%20Structure%20%28by%20ref%29))
  (bind _pawn (Variables|FortGameplayCueNotifyLoop_LowGravity|GetPawn))
  (bind _mesh (Variables|/Script/Engine.Character|GetMesh _pawn))
  (Actor|Tick|SetActorTickEnabled false)
  (CallFunction|ClearExpireTell)
  (bind _loopingfx (Variables|Default|GetLoopingFX)
    (:then
      (Components|Activation|Deactivate _loopingfx)
      (Effects|SpawnFXSystemAttached "/FNGameplayCues/Athena/Items/ForagedItems/LowGravity/FX/NS_LowGravity_End.NS_LowGravity_End" _mesh))
    (:else
      _))
  (Audio|SpawnSoundAttached (Variables|Sounds|GetDeactivateSound) _mesh)
  (Audio|Components|Audio|FadeOut (Variables|Default|GetInUse%20Audio%20Comp) 0.1 0.0 "Linear" :!class /Script/Engine.AudioComponent)
  (return true))

(fn ConstructionScript)

(fn OnBecomeRelevant
  (params (MyTarget :type Actor%20Object%20Reference) (Parameters :type Gameplay%20Cue%20Parameters%20Structure%20%28by%20ref%29))
  (bind _inuse_audio_comp (Variables|Default|GetInUse%20Audio%20Comp))
  (bind _pawn (Variables|FortGameplayCueNotifyLoop_LowGravity|GetPawn))
  (bind _mesh (Variables|/Script/Engine.Character|GetMesh _pawn))
  (CallFunction|Activated MyTarget)
  (Actor|Tick|SetActorTickEnabled true)
  (bind _returnvalue (Effects|SpawnFXSystemAttached (Variables|Default|GetLowGravFXSystem%20Asset) _mesh))
  (Variables|Default|SetLoopingFX _returnvalue)
  (Audio|SpawnSoundAttached (Variables|Sounds|GetActivateSound) _mesh)
  (Audio|Parameter|SetBooleanParameter _inuse_audio_comp "IsLocal" (Pawn|IsLocallyControlled (Variables|FortGameplayCueNotifyLoop_LowGravity|GetPawn)) :!class /Script/Engine.AudioComponent)
  (Audio|Components|Audio|FadeIn _inuse_audio_comp 0.1 1.0 0.0 "Linear" :!class /Script/Engine.AudioComponent)
  (return false))
