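; Gameplay cue notify actor for a pool floaty ability or effect on the Athena player.
; When the cue becomes relevant, it calls Activated, which stores the target as PlayerPawn, enables actor tick, sets a fixed relative rotation on the Arrow component and sets IsActive.
; Ceasing relevance disables tick but leaves IsActive set.
; No tick logic appears in this graph, and the construction script is empty.
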(fn Activated
  (params (PlayerPawn :type Actor%20Object%20Reference))
  (Variables|Default|SetPlayerPawn PlayerPawn)
  (Actor|Tick|SetActorTickEnabled true)
  (Transformation|SetRelativeRotation (Variables|Default|GetArrow) "0.000000,90.000000,-90.000000" false false)
  (Variables|Default|SetIsActive true))

(fn OnCeaseRelevant
  (params (MyTarget :type Actor%20Object%20Reference) (Parameters :type Gameplay%20Cue%20Parameters%20Structure%20%28by%20ref%29))
  (Actor|Tick|SetActorTickEnabled false)
  (return false))

(fn ConstructionScript)

(fn OnBecomeRelevant
  (params (MyTarget :type Actor%20Object%20Reference) (Parameters :type Gameplay%20Cue%20Parameters%20Structure%20%28by%20ref%29))
  (CallFunction|Activated MyTarget)
  (return false))
