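; Looping gameplay cue for the Sneaky Snowman that shows a snowman mesh on the affected player and adds custom footstep audio.
; On loop start it snaps itself to the player's pelvis, rotates its arrow, reveals the mesh and triggers a time-based material wobble.
; It binds to the pawn's footstep event to play a sound at the player's location, then hides the mesh and unbinds on removal.
; BeginPlay still prints a "NEW CUE" debug string.
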(event EventBeginPlay
  (Development|PrintString 0 "NEW CUE" true true "(R=1.000000,G=0.000000,B=0.121054,A=1.000000)" 1e+01))

(event EventOnLoopingStart(Cascade)
  (params (MyTarget :type Actor%20Object%20Reference) (Parameters :type Gameplay%20Cue%20Parameters%20Structure%20%28by%20ref%29) (ParticleComponents :type Array%20of%20Cascade%20Particle%20System%20Component%20Object%20References) (AudioComponents :type Array%20of%20Audio%20Component%20Object%20References))
  (bind _snowmanmesh (Variables|Default|GetSnowmanMesh))
  (Variables|Components|SetPlayerPawn MyTarget)
  (Actor|Tick|SetActorTickEnabled true)
  (Transformation|SetRelativeRotation (Variables|Default|GetArrow) "0.000000,90.000000,-90.000000" false false)
  (Rendering|SetVisibility _snowmanmesh true false)
  (Rendering|Material|SetScalarParameterValueonMaterials _snowmanmesh "WobbleStartTime" (Utilities|Time|GetGameTimeinSeconds 0))
  (Rendering|Material|SetScalarParameterValueonMaterials _snowmanmesh "WobbleStrength" 1.0)
  (bind _asfort_player_pawn_athena (Utilities|Casting|CastToFortPlayerPawnAthena (Variables|Components|GetPlayerPawn)))
  (EventDispatchers|BindEventtoOnFootstepEvent _asfort_player_pawn_athena (EventDispatchers|CreateEvent :!event Custom|On%20Player%20Step))
  (Transformation|AttachActorToComponent (Variables|/Script/Engine.Character|GetMesh _asfort_player_pawn_athena) "pelvis" "SnapToTarget" "SnapToTarget" "KeepWorld"))

(event EventOnRemoval(Cascade)
  (params (MyTarget :type Actor%20Object%20Reference) (Parameters :type Gameplay%20Cue%20Parameters%20Structure%20%28by%20ref%29) (ParticleComponents :type Array%20of%20Cascade%20Particle%20System%20Component%20Object%20References) (AudioComponents :type Array%20of%20Audio%20Component%20Object%20References) (BurstCameraShakeInstance :type Camera%20Shake%20Base%20Object%20Reference) (BurstDecalInstance :type Decal%20Actor%20Object%20Reference))
  (Actor|Tick|SetActorTickEnabled false)
  (Rendering|SetVisibility (Variables|Default|GetSnowmanMesh) false false)
  (bind _playerpawn (Variables|Components|GetPlayerPawn)
    (:then
      (bind _asfort_player_pawn_athena (Utilities|Casting|CastToFortPlayerPawnAthena _playerpawn))
      (EventDispatchers|UnbindEventfromOnFootstepEvent _asfort_player_pawn_athena (EventDispatchers|CreateEvent :!event Custom|On%20Player%20Step)))
    (:else)))

(event Custom|On%20Player%20Step
  (params (FootIndex :type Integer))
  ;[1 1 1] On Player Step
  (Utilities|Casting|CastToFortPlayerPawnAthena (Variables|Components|GetPlayerPawn))
  (Audio|PlaySoundatLocation (Variables|Components|GetPlayerPawn) (Variables|Default|GetFootstepSound) (Transformation|GetActorLocation (Variables|Components|GetPlayerPawn))))

(fn ConstructionScript)
